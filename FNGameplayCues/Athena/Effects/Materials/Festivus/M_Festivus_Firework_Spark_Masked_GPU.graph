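# Unlit masked particle material for Festivus firework sparks on Niagara or particle sprites, with responsive AA and a round SphereMask dot shape.
# Particle color is hue-shifted by world position, then by per-particle random times relative age, and passed through emissive adjustments.
# The mobile shading path skips both hue shifts and uses plain particle color.
# MF_SmoothLod handles near and far distance fade scaling, randomizing the far end scale per particle.
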Material /FNGameplayCues/Athena/Effects/Materials/Festivus/M_Festivus_Firework_Spark_Masked_GPU
graph {
 properties {
  bEnableResponsiveAA=true
  bUsedWithNiagaraSprites=true
  bUsedWithParticleSprites=true
  blendMode=BLEND_Masked
  opacityMaskClipValue=0.1
  pixelDepthOffsetMode=PDOM_Legacy
  shadingModel=MSM_Unlit
  translucencyLightingMode=TLM_VolumetricPerVertexDirectional
  usedShadingModels=Unlit
 }
 nodes [
  n1 = Multiply(A=ParticleRandom(), B=Multiply(A=ParticleRelativeTime(), B=ScalarParameter() {defaultValue=0.25, parameterName=HueShiftValue}))
  n2 = ParticlePositionWS()
  n3 = ScalarParameter() {defaultValue=5000, parameterName=ParticlePostion_ColorScale}
  n4 = MaterialFunctionCall(
   Hue Shift Percentage=Multiply(A=ScalarParameter() {defaultValue=0.5, parameterName=HueShiftValue_WholeSystem}, B=Divide(A=@n2, B=@n3) {constB=50000})
   Texture=ParticleColor().RGB
  ) {materialFunction=&HueShift}
  MaterialFunctionCall(Emissive=ShadingPathSwitch(
    Default=MaterialFunctionCall(Hue Shift Percentage=@n1, Texture=@n4.Result) {materialFunction=&HueShift}.Result
    Mobile=ParticleColor().RGB
   ) {desc=Migrated from FeatureLevelSwitch}) {materialFunction=&MF_FX_EmissiveAdjustments}
  MaterialFunctionCall!(
   NearFadeLength=ScalarParameter() {defaultValue=1500, parameterName="NearFadeLength (S)"}
   NearFadeOffset
   NearFadeCurve
   NearFadeStartingScale=ScalarParameter() {defaultValue=1, parameterName="NearFadeStartingScale (S)"}
   NearFadeEndingScale
   FarFadeLength=ScalarParameter() {defaultValue=50000, parameterName="FarFadeLength (S)"}
   FarFadeOffset
   FarFadeCurve
   FarFadeEndingScale=LinearInterpolate(
    A=ScalarParameter() {defaultValue=6, parameterName=FarFadeEndingScale_LOW}
    B=ScalarParameter() {defaultValue=12, parameterName=FarFadeEndingScale_HIGH}
    Alpha=ParticleRandom()
   )
   Opacity Near Distance Bias
   Opacity Far Distance Bias
  ) {materialFunction=/Game/Effects/Fort_Effects/Materials/_Functions/MF_SmoothLod}
  SphereMask(
   A=TextureCoordinate()
   B=Constant() {r=0.5}
   Radius=ScalarParameter() {defaultValue=0.5, parameterName=Radius}
   Hardness=ScalarParameter() {defaultValue=0.2, parameterName=Hardness}
  )
 ]
}
